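# Material function for the Robro character's animated screen-like eyes, overriding incoming material attributes and adding emissive.
# Thresholded facial pose parameters (squint, brow, glabella, blink) pick flipbook cells by offsetting UV1 through chained lerps, with blink taking priority.
# Hashed per-line noise periodically shifts U for glitches before sampling the tinted eye texture.
# View-dependent sine scanlines with a brighter panning band add emissive and modulate the eye glow.
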MaterialFunction /BRCosmetics/Characters/Player/Male/Medium/Bodies/M_MED_Robro/Materials/MF_Robroeyes
graph {nodes [
  n1 = ComponentMask(None=TextureCoordinate() {coordinateIndex=1}) {g=true}
  n2 = Multiply(A=Time(), B=ScalarParameter() {defaultValue=-0.1, parameterName=ScanLineSpeed})
  Add(A=@n1, B=@n2)
  n3 = ComponentMask(None=TextureCoordinate() {coordinateIndex=1}) {g=true}
  Add(A=@n3, B=Multiply(A=Time(), B=ScalarParameter() {defaultValue=-0.1, parameterName=RB_GlitchLineSpeed}))
  n4 = GetMaterialAttributes(None=FunctionInput() {inputName=MaterialAttributes, inputType=FunctionInput_MaterialAttributes}) {attributeGetTypes [#1]}
  n5 = TextureCoordinate() {coordinateIndex=1}
  n6 = Add(A=@n5, B=Constant2Vector() {g=0.75})
  n7 = Add(A=@n5, B=Constant2Vector() {r=0.5})
  n8 = ScalarParameter() {defaultValue=0.5, group=Banana_Poses, parameterName=R_squint_inner_pose}
  n9 = Divide(B=ScalarParameter() {defaultValue=0.5, group=Banana, parameterName=Threshold_squint_inner_pose}) {constA=0.5}
  n10 = Clamp(None=Round(None=Multiply(A=@n8, B=@n9)))
  n11 = Constant2Vector() {g=0.5}
  n12 = ScalarParameter() {defaultValue=0.3, group=Banana, parameterName=Threshold_R_brow_up_pose}
  n13 = Round(None=Multiply(A=ScalarParameter() {group=Banana_Poses, parameterName=R_brow_up_pose}, B=Divide(B=@n12) {constA=0.5}))
  n14 = Multiply(
   A=ScalarParameter() {group=Banana_Poses, parameterName=C_glabella_down_pose}
   B=Divide(B=ScalarParameter() {defaultValue=0.2, group=Banana, parameterName=Threshold_Glabella_down}) {constA=0.5}
  )
  n15 = LinearInterpolate(
   A=LinearInterpolate(A=LinearInterpolate(A=@n6, B=@n7, Alpha=@n10), B=Add(A=@n5, B=@n11), Alpha=Clamp(None=@n13))
   B=Add(A=@n5, B=Constant2Vector() {g=0.5, r=0.5})
   Alpha=Clamp(None=Round(None=@n14))
  )
  n16 = Add(A=@n5, B=Constant2Vector() {g=0.25, r=0.5})
  n17 = ScalarParameter() {group=Banana_Poses, parameterName=C_glabella_up_pose}
  n18 = Divide(B=ScalarParameter() {defaultValue=0.5, group=Banana, parameterName=Threshold_Glabella_up}) {constA=0.5}
  n19 = Clamp(None=Round(None=Multiply(A=@n17, B=@n18)))
  n20 = Constant2Vector() {g=0.25}
  n21 = ScalarParameter() {defaultValue=0.3, group=Banana, parameterName=Threshold_R_blink_pose}
  n22 = Round(None=Multiply(A=ScalarParameter() {group=Banana_Poses, parameterName=R_blink_pose}, B=Divide(B=@n21) {constA=0.5}))
  n23 = MaterialFunctionCall(Float2=LinearInterpolate(A=LinearInterpolate(A=@n15, B=@n16, Alpha=@n19), B=Add(A=@n5, B=@n20), Alpha=Clamp(None=@n22))) {materialFunction=&BreakOutFloat2Components}
  n24 = Ceil(None=Multiply(A=@n3, B=ScalarParameter() {defaultValue=100, parameterName=RB_GlitchLine_Tiling}))
  n25 = ComponentMask(None=MaterialFunctionCall(Seed=Time(), Float=ComponentMask(None=@n24) {g=true}, Vec2=@n24) {materialFunction=&MF_Hash}.Vec2) {r=true}
  n26 = ScalarParameter() {defaultValue=1, parameterName=RB_GlitchDistortion_Amt}
  n27 = Frac(None=Multiply(A=Time(), B=ScalarParameter() {defaultValue=0.1, parameterName=RB_GlitchFrequency}) {constB=0.05})
  n28 = MaterialFunctionCall(Float=Time()) {materialFunction=&MF_Hash}
  n29 = LinearInterpolate(
   A=ScalarParameter() {defaultValue=0.8, parameterName=RB_GlitchThreshold_Min}
   B=ScalarParameter() {defaultValue=0.9, parameterName=RB_GlitchThreshold_Max}
   Alpha=Round(None=ConstantBiasScale(None=@n28.Float))
  )
  n30 = Saturate(None=Ceil(None=Subtract(A=@n27, B=@n29)))
  n31 = Add(A=@n23.R, B=Multiply(A=Multiply(A=@n25, B=@n25), B=Multiply(A=@n26, B=@n30)))
  n32 = Multiply(
   A=VectorParameter() {parameterName=RB_EyeColor, defaultValue=<rgba 0.8796 0.2542 0.8148 1>}.RGB
   B=TextureSample(UVs=MaterialFunctionCall(X=@n31, Y=@n23.G) {materialFunction=&MakeFloat2}.Result) {texture=/BRCosmetics/Characters/Player/Male/Medium/Bodies/M_MED_Robro/Textures/T_Robro_Eyes_FlipBook}.R
  )
  n33 = CameraVectorWS()
  n34 = VertexNormalWS()
  n35 = Multiply(A=@n1, B=ScalarParameter() {defaultValue=50, parameterName=RB_ScanLine_Tiling})
  n36 = Clamp(None=Multiply(A=DotProduct(A=@n33, B=@n34), B=Sine(None=@n35)))
  n37 = Multiply(A=VectorParameter() {parameterName=RB_ScanLine_Color, defaultValue=<rgba 0.4403 0.183 0.8796 1>}.RGB, B=@n36)
  n38 = Add(A=@n1, B=@n2)
  n39 = LinearInterpolate(
   A=Multiply(A=@n37, B=ScalarParameter() {defaultValue=0.15, parameterName=RB_ScanLine_Brightness})
   B=Multiply(A=@n37, B=ScalarParameter() {defaultValue=0.5, parameterName=RB_ScanLine_Pan_Brightness})
   Alpha=SphereMask(A=Frac(None=@n38), B=Constant() {r=0.5}, Radius=ScalarParameter() {defaultValue=0.1, parameterName=RB_Scanline_Radius}, Hardness=Constant())
  )
  n40 = ScalarParameter() {defaultValue=0.1, parameterName=RB_EyeGlowStrength_Min}
  n41 = ScalarParameter() {defaultValue=2, parameterName=RB_EyeGlowStrength_Max}
  n42 = Power(Base=Multiply(A=@n32, B=LinearInterpolate(A=@n40, B=@n41, Alpha=@n36)))
  n43 = SetMaterialAttributes!(
   MaterialAttributes=@n4.MaterialAttributes
   BaseColor=@n32
   Roughness=ScalarParameter() {defaultValue=0.25, parameterName=RB_Roughness}
   EmissiveColor=Add(A=@n4.EmissiveColor, B=Add(A=@n39, B=@n42))
   Metallic=Constant()
   Specular=Constant() {r=1}
  ) {attributeSetTypes [
    #2
    #3
    #1
    #4
    #5
   ]}
  FunctionOutput(None=@n43)
  Reroute(None=@n5)
 ]}
