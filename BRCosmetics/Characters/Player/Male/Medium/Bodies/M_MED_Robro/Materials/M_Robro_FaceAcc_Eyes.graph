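# Material for the eye face accessory on the Robro character body, set up for skeletal meshes and mesh deformers.
# It uses material attributes, passing the shared character PreFX result through MF_Robroeyes and then into the shared PostFX function.
# The PreFX AO output feeds PostFX's AO, Distance Blend and Fresnel inputs alike.
# It uses Default Lit shading and disables ray-traced shadow casting.
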Material /BRCosmetics/Characters/Player/Male/Medium/Bodies/M_MED_Robro/Materials/M_Robro_FaceAcc_Eyes
graph {
 properties {
  bCastRayTracedShadows=false
  bUseMaterialAttributes=true
  bUsedWithMeshDeformer=true
  bUsedWithSkeletalMesh=true
  pixelDepthOffsetMode=PDOM_Legacy
  usedShadingModels=Default Lit
 }
 nodes [
  n1 = MaterialFunctionCall() {materialFunction=&BaseFN_CharacterFunction_PreFX}
  MaterialFunctionCall(
   AO=@n1.AO
   Distance Blend=@n1.AO
   Fresnel=@n1.AO
   Material=MaterialFunctionCall!(MaterialAttributes=@n1.Result) {materialFunction=&/MF_Robroeyes}.Result
  ) {materialFunction=&BaseFN_CharacterFunction_PostFX}
 ]
}
