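# Physics asset for the legs mesh of the male medium player cosmetic body Robro, defining secondary-motion bodies for shorts, shoelaces and a bowtie.
# Capsule bodies sit on dynamic bones, with foot and upper-spine bodies as anchors.
# Four shorts bones per side hang from a master shorts body with wide swing and twist limits.
# Shoelaces swing from the feet without twist, while the bowtie and its two sides are tightly limited.
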PhysicsAsset /BRCosmetics/Characters/Player/Male/Medium/Bodies/M_MED_Robro/Mesh/M_MED_Robro_Legs_Physics
bodies {
 dyn_bowtie [{shapeType=Capsule, length=1, radius=1, center=<0.01 0 0>}]
 dyn_bowtie_l [{shapeType=Capsule, length=0.5, radius=0.5, center=<0.1 0 0>, rotation=<rot 90 0 0>}]
 dyn_bowtie_r [{shapeType=Capsule, length=0.5, radius=0.5, center=<0.1 0 0>, rotation=<rot 90 0 0>}]
 dyn_master_shorts_l [
  {shapeType=Capsule, length=10, radius=5, center=<0 1.6493 0.3513>, rotation=<rot 90 0 0>}
  {shapeType=Capsule, length=10, radius=5, center=<0 -0.8274 0>, rotation=<rot 90 0 0>}
 ]
 dyn_master_shorts_r [{shapeType=Capsule, length=10, radius=5, center=<0.2549 -0.8059 0>}, {shapeType=Capsule, length=10, radius=5, center=<0 1.771 0>}]
 dyn_shoelaces_l [{shapeType=Capsule, length=2, radius=1.5}]
 dyn_shoelaces_r [{shapeType=Capsule, length=2, radius=1.5}]
 dyn_shorts_b_l [{shapeType=Capsule, length=2, radius=2}]
 dyn_shorts_b_r [{shapeType=Capsule, length=2, radius=2}]
 dyn_shorts_f_l [{shapeType=Capsule, length=2, radius=2}]
 dyn_shorts_f_r [{shapeType=Capsule, length=2, radius=2}]
 dyn_shorts_l_l [{shapeType=Capsule, length=2, radius=2}]
 dyn_shorts_l_r [{shapeType=Capsule, length=2, radius=2}]
 dyn_shorts_r_l [{shapeType=Capsule, length=2, radius=2}]
 dyn_shorts_r_r [{shapeType=Capsule, length=2, radius=2}]
 foot_l [{shapeType=Capsule, length=2, radius=2, center=<3.9449 3.8238 -0.0186>, rotation=<rot 0.3964 0.4099 0.6988>}]
 foot_r [{shapeType=Capsule, length=2, radius=2, center=<-3.5742 -3.8868 0>}]
 spine_05 [{shapeType=Capsule, length=0.1, radius=0.1, center=<2.9297 1.4952 -0.217>, rotation=<rot 0 -161.2448 -87.4903>}]
}
constraints [
 {bone1Name=dyn_bowtie_r, bone2Name=dyn_bowtie, swing1LimitDegrees=15, swing2LimitDegrees=15, twistLimitDegrees=5}
 {bone1Name=dyn_bowtie_l, bone2Name=dyn_bowtie, swing1LimitDegrees=15, swing2LimitDegrees=15, twistLimitDegrees=5}
 {bone1Name=dyn_shoelaces_r, bone2Name=foot_r, swing1LimitDegrees=45, swing2LimitDegrees=45, twistLimitDegrees=0}
 {bone1Name=dyn_shoelaces_l, bone2Name=foot_l, swing1LimitDegrees=45, swing2LimitDegrees=45, twistLimitDegrees=0}
 {bone1Name=dyn_shorts_f_l, bone2Name=dyn_master_shorts_l, swing1LimitDegrees=45, swing2LimitDegrees=45, twistLimitDegrees=45}
 {bone1Name=dyn_shorts_b_l, bone2Name=dyn_master_shorts_l, swing1LimitDegrees=45, swing2LimitDegrees=45, twistLimitDegrees=45}
 {bone1Name=dyn_shorts_l_l, bone2Name=dyn_master_shorts_l, swing1LimitDegrees=45, swing2LimitDegrees=45, twistLimitDegrees=45}
 {bone1Name=dyn_shorts_r_l, bone2Name=dyn_master_shorts_l, swing1LimitDegrees=45, swing2LimitDegrees=45, twistLimitDegrees=45}
 {bone1Name=dyn_shorts_l_r, bone2Name=dyn_master_shorts_r, swing1LimitDegrees=45, swing2LimitDegrees=45, twistLimitDegrees=45}
 {bone1Name=dyn_shorts_r_r, bone2Name=dyn_master_shorts_r, swing1LimitDegrees=45, swing2LimitDegrees=45, twistLimitDegrees=45}
 {bone1Name=dyn_shorts_f_r, bone2Name=dyn_master_shorts_r, swing1LimitDegrees=45, swing2LimitDegrees=45, twistLimitDegrees=45}
 {bone1Name=dyn_shorts_b_r, bone2Name=dyn_master_shorts_r, swing1LimitDegrees=45, swing2LimitDegrees=45, twistLimitDegrees=45}
 {bone1Name=dyn_bowtie, bone2Name=spine_05, swing1LimitDegrees=10, swing2LimitDegrees=10, twistLimitDegrees=5}
]
