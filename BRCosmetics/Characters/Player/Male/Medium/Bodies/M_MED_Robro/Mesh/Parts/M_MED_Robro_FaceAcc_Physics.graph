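# Physics asset for the face accessory part of the male medium Robro player body, covering dynamic hair and glasses-band bones.
# Four small hair capsules attach to the head with limited swing and twist.
# Three glasses-band bodies, each two capsules, hang off a master band body with tighter twist limits; that master joins the head, and the head the neck, with free angular motion.
# Unconstrained clavicle capsules and a neck capsule supply upper-body collision shapes.
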PhysicsAsset /BRCosmetics/Characters/Player/Male/Medium/Bodies/M_MED_Robro/Mesh/Parts/M_MED_Robro_FaceAcc_Physics
bodies {
 clavicle_l [{shapeType=Capsule, length=10.4683, radius=7.4947, center=<6.7413 -1.363 2.2663>, rotation=<rot 82.0372 0 180>}]
 clavicle_r [{shapeType=Capsule, length=10.4683, radius=7.4947, center=<-6.3234 1.3842 -2.2382>, rotation=<rot 82.0401 0 0>}]
 dyn_glasses_band1 [
  {shapeType=Capsule, length=4.3832, radius=1, center=<5 -0.0001 0>, rotation=<rot 90 0 0>}
  {shapeType=Capsule, length=5.679, radius=0.9091, center=<12.7125 -1.7631 0>}
 ]
 dyn_glasses_band_l [
  {shapeType=Capsule, length=3.6341, radius=1, center=<5 -0.0001 0.3005>, rotation=<rot 84.2365 -179.9999 180>}
  {shapeType=Capsule, length=3.7932, radius=1, center=<9.7547 -0.0002 -0.0128>, rotation=<rot 59.9638 0.0003 27.6992>}
 ]
 dyn_glasses_band_r [
  {shapeType=Capsule, length=4.3832, radius=1, center=<5 -0.0001 -0.2729>, rotation=<rot 84.8106 0 0>}
  {shapeType=Capsule, length=2.7516, radius=1, center=<9.3828 0 -0.3799>, rotation=<rot 61.4653 -179.9999 -163.7861>}
 ]
 dyn_hair_01 [{shapeType=Capsule, length=1, radius=1, center=<-0.1 0 0>}]
 dyn_hair_02 [{shapeType=Capsule, length=1, radius=1, center=<0.1 0 0>}]
 dyn_hair_03 [{shapeType=Capsule, length=1, radius=1, center=<-0.1 0 0>}]
 dyn_hair_04 [{shapeType=Capsule, length=1, radius=1, center=<-0.1 0 0>}]
 dyn_master_glasses_band [{shapeType=Capsule, length=0.505, radius=0.505}]
 head [{shapeType=Capsule, length=1, radius=1}]
 neck_01 [{shapeType=Capsule, length=3.064, radius=4.538, center=<0 -0.9451 0>, rotation=<rot 90 0 -6.738>}]
}
constraints [
 {bone1Name=dyn_hair_01, bone2Name=head, swing1LimitDegrees=45, swing2LimitDegrees=45, twistLimitDegrees=45}
 {bone1Name=dyn_hair_04, bone2Name=head, swing1LimitDegrees=45, swing2LimitDegrees=45, twistLimitDegrees=45}
 {bone1Name=dyn_hair_03, bone2Name=head, swing1LimitDegrees=40, swing2LimitDegrees=40, twistLimitDegrees=10}
 {bone1Name=dyn_hair_02, bone2Name=head, swing1LimitDegrees=45, swing2LimitDegrees=45, twistLimitDegrees=45}
 {
  bone1Name=dyn_master_glasses_band
  bone2Name=head
  swing1LimitDegrees=45
  swing1Motion=Free
  swing2LimitDegrees=45
  swing2Motion=Free
  twistLimitDegrees=45
  twistMotion=Free
 }
 {bone1Name=dyn_glasses_band1, bone2Name=dyn_master_glasses_band, swing1LimitDegrees=35, swing2LimitDegrees=35, twistLimitDegrees=7}
 {bone1Name=dyn_glasses_band_l, bone2Name=dyn_master_glasses_band, swing1LimitDegrees=35, swing2LimitDegrees=35, twistLimitDegrees=7}
 {bone1Name=dyn_glasses_band_r, bone2Name=dyn_master_glasses_band, swing1LimitDegrees=35, swing2LimitDegrees=35, twistLimitDegrees=7}
 {
  bone1Name=head
  bone2Name=neck_01
  swing1LimitDegrees=45
  swing1Motion=Free
  swing2LimitDegrees=45
  swing2Motion=Free
  twistLimitDegrees=45
  twistMotion=Free
 }
]
