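# Animation Blueprint for the Robro character's face accessory part, driving glasses band motion and the smile mouth material.
# It copies the attached parent's pose and curves, additively lifts the band bone while crouch-moving, and aims it at dyn_back_master.
# A RigidBody node then simulates physics relative to spine_05 with damped component motion.
# Each update sets the crouch-moving flag and switches the mouth animation frame on a dynamic material according to emote state.
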AnimBlueprint /BRCosmetics/Characters/Player/Male/Medium/Bodies/M_MED_Robro/Mesh/Parts/M_MED_Robro_FaceAcc_Skeleton_AnimBlueprint
target_skeleton=&/M_MED_Robro_FaceAcc_Skeleton
graphs [
 AnimGraph {nodes [
   n1 = "Transform (Modify) Bone - Bone: dyn_master_glasses_band"!(
    ComponentPose=Local To Component(LocalPose=Copy Pose From Mesh() {bCopyCurves=true, bUseAttachedParent=true})
    bAlphaBoolEnabled:bool=Get crouchMoving() -> (crouchMoving:bool)
    Translation="-0.000,0.000,2.5"
    Rotation="0,0,0"
    Scale="1,1,1"
   ) -> (Pose) {
    alphaInputType=Bool
    rotationMode=BMM_Additive
    translationMode=BMM_Additive
    alphaBoolBlend {blendInTime=0.1, blendOption=Linear, blendOutTime=0.1}
    boneToModify {boneName=dyn_master_glasses_band}
    translation {z=2.5}
   }
   n2 = Look At - Bone: dyn_master_glasses_band(ComponentPose=@n1, Alpha=1) {
    alphaBoolBlend {blendOption=Linear}
    boneToModify {boneName=dyn_master_glasses_band}
    lookAtLocation {z=50}
    lookAtTarget {boneReference {boneName=dyn_back_master}}
    lookAt_Axis {axis=<-1 0>}
    lookUp_Axis {axis=<0 1>}
   }
   n3 = RigidBody(ComponentPose=@n2) {
    bClampLinearTranslationLimitToRefPose=true
    bOverrideWorldGravity=true
    overridePhysicsAsset=&/M_MED_Robro_FaceAcc_Physics
    simulationSpace=BaseBoneSpace
    alphaBoolBlend {blendOption=Linear}
    baseBoneRef {boneName=spine_05}
    componentAppliedLinearAccClamp=<1500 1000 1000>
    componentLinearAccScale=<0.3 0.15 0.15>
    componentLinearVelScale=<0.3 0.15 0.15>
    overrideWorldGravity {z=-980}
   }
   Output Pose(Result=Component To Local(ComponentPose=@n3)) {layerGroup=None}
  ]}
 EventGraph {nodes [
   Event BlueprintUpdateAnimation()
   n1 = Get PawnSpeed2D()
   > Set crouchMoving(crouchMoving:bool=AND Boolean(A=Get bIsCrouching(), B="float > float"(A=@n1, B=1) {is=title K2Node_CallFunction})) -> (Output_Get:bool)
   > n2 = Get Dynamic Material Instance(SkeletalMeshComponent=GetPartSkeletalMeshComponent(), ElementIndex=1)
   > n3 = Branch(Condition=Get bIsPlayingEmote())
   > SetScalarParameterValue(self=@n2, ParameterName=Smile Mouth animation Frame, Value=8)
   SetScalarParameterValue(execute=@n3.else, self=@n2, ParameterName=Smile Mouth animation Frame, Value=3)
   TryGetPawnOwner()
  ]}
]
variables [crouchMoving {type=boolean}]
